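# Unlit additive material for the glow on the Zero Point river flat mesh, with animated, tinted plasma swirls.
# Panning radial UVs, warped by noise, drive three offset plasma samples packed into RGB for color fringing, plus a hand-painted base layer.
# Opacity combines radial masks, time-of-day scaling and depth fade, which Low quality skips.
# Cutout masks, pixel-depth distance fade, optional Fresnel, a global water-opacity collection multiplier and camera-offset nodes are also present.
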Material /Game/Environments/Hera/FX/Materials/ZeroPoint/M_ZeroPoint_RiverFlatmesh_Glow
graph {
 properties {
  bAutomaticallySetUsageInEditor=false
  blendMode=BLEND_Additive
  previewMesh=/Game/Environments/Hera/FX/Meshes/SM_ZeroPoint_RiverFlatMesh_01
  shadingModel=MSM_Unlit
  usedShadingModels=Unlit
 }
 nodes [
  Constant()
  Constant() {r=0.45}
  MaterialFunctionCall(Offset Amount=ScalarParameter() {defaultValue=-33, parameterName=Param_2}) {materialFunction=/Engine/Functions/Engine_MaterialFunctions02/WorldPositionOffset/CameraOffset}
  MaterialFunctionCall(Emissive=NamedRerouteUsage()) {materialFunction=&MF_FX_EmissiveAdjustments}
  n1 = NamedRerouteUsage()
  n2 = VectorParameter() {parameterName=FresnelOpacityParams, channelNames {a=Enable, b=InvertExp, g=Base, r=Exp}, defaultValue=<rgba 5 0 2 0>}
  n3 = Power(Base=OneMinus(None=Fresnel(ExponentIn=@n2.Exp, BaseReflectFractionIn=@n2.Exp) {baseReflectFraction=0}), Exp=@n2.InvertExp)
  Multiply(
   A=LinearInterpolate(A=@n1, B=Multiply(A=@n1, B=@n3), Alpha=@n2.Enable)
   B=CollectionParameter() {collection=&/MPC_ZP_WaterOpacity, parameterName=GlobalOpacity_Mult}
  )
  n4 = MaterialFunctionCall() {materialFunction=&VectorToRadialValue}
  n5 = ComponentMask(None=VectorParameter() {parameterName=Speed, defaultValue=<rgba 0.02 -0.02 0 0>}.RGB) {g=true, r=true}
  n6 = ComponentMask(None=VectorParameter() {parameterName=noise tiling, defaultValue=<rgba 10 10 0 1>}.RGB) {g=true, r=true}
  n7 = Panner(
   Coordinate=MaterialFunctionCall() {materialFunction=&VectorToRadialValue}.Radial Coordinates
   Speed=VectorParameter() {parameterName=noise speed, defaultValue=<rgba 0 -0.02 -1 1>}.RGB
  ) {speedX=0.2, speedY=-1}
  n8 = VectorParameter() {parameterName=noise offset, defaultValue=<rgba 0 0.1 0 1>}
  n9 = Constant()
  n10 = TextureSample!(UVs=Add(A=Multiply(A=@n6, B=@n7), B=ComponentMask(None=@n8.RGB) {g=true, r=true}), Tex, "DDX(UVs)"=@n9, "DDY(UVs)"=@n9, Apply View MipBias) {
   mipValueMode=TMVM_Derivative
   samplerType=SAMPLERTYPE_LinearGrayscale
   texture=/Game/Effects/Fort_Effects/Emotes/UltraLight/Texture/T_DirectionalNoise_Greyscale
  }
  n11 = Multiply(A=@n10.R, B=ScalarParameter() {defaultValue=0.05, parameterName=Noise_Intensity}) {constB=0.2}
  n12 = Add(A=Multiply(A=Panner(Coordinate=@n4.Radial Coordinates, Speed=@n5) {speedY=-0.4}, B=Constant2Vector() {g=1, r=1}), B=@n11)
  n13 = ScalarParameter() {defaultValue=0.02, parameterName=Color offset}
  n14 = Constant()
  n15 = TextureSampleParameter2D!(UVs=Add(A=@n13, B=@n12), "DDX(UVs)"=@n14, "DDY(UVs)"=@n14, Apply View MipBias) {
   mipValueMode=TMVM_Derivative
   parameterName=Tex Noise
   samplerType=SAMPLERTYPE_LinearGrayscale
   texture=/Game/Effects/Fort_Effects/Emotes/UltraLight/Texture/T_PlasmaCoils
  }
  n16 = Constant()
  n17 = TextureSampleParameter2D(UVs=@n12, "DDX(UVs)"=@n16, "DDY(UVs)"=@n16) {
   mipValueMode=TMVM_Derivative
   parameterName=Tex Noise
   samplerType=SAMPLERTYPE_LinearGrayscale
   texture=&T_PlasmaCoils
  }
  n18 = Constant()
  n19 = TextureSampleParameter2D(UVs=Subtract(A=@n12, B=@n13), "DDX(UVs)"=@n18, "DDY(UVs)"=@n18) {
   mipValueMode=TMVM_Derivative
   parameterName=Tex Noise
   samplerType=SAMPLERTYPE_LinearGrayscale
   texture=&T_PlasmaCoils
  }
  n20 = Desaturation(
   None=Power(Base=MaterialFunctionCall(X=@n15.R, Y=@n17.R, Z=@n19.R) {materialFunction=&MakeFloat3}.Result) {constExponent=3}
   Fraction=ScalarParameter() {defaultValue=-0.2, parameterName=Des}
  )
  n21 = Constant()
  n22 = TextureSampleParameter2D(UVs=@n12, "DDX(UVs)"=@n21, "DDY(UVs)"=@n21) {
   mipValueMode=TMVM_Derivative
   parameterName=Tex Base
   samplerType=SAMPLERTYPE_LinearGrayscale
   texture=/Game/Characters/Player/Male/Medium/Bodies/M_MED_Citadel/FX/T_Citadel_Noise_Dissolve_Handpaint
  }
  n23 = Multiply(A=@n22.R, B=ScalarParameter() {defaultValue=10, parameterName=Center Alpha}) {constB=0.7}
  NamedRerouteDeclaration(None=Multiply(A=Add(A=@n20, B=@n23), B=VectorParameter() {parameterName=Color, defaultValue=<rgba 1 0.0885 1 0>}.RGB)) {name=Color, nodeColor=<rgba 0.8281 1 0 1>}
  n24 = Power(Base=@n17.R, Exp=ScalarParameter() {defaultValue=1, parameterName=Alpha power}) {constExponent=3}
  n25 = Add(A=TextureCoordinate(), B=Constant2Vector() {g=-0.07, r=-0.1})
  n26 = Power(
   Base=SphereMask(A=Add(A=@n25, B=@n11), B=Constant2Vector() {g=0.458, r=0.444}) {attenuationRadius=0.9, hardnessPercent=0}
   Exp=ScalarParameter() {defaultValue=5, parameterName=Param_3}
  ) {constExponent=4}
  n27 = ComponentMask(None=VectorParameter() {parameterName=noise tiling, defaultValue=<rgba 10 10 0 1>}.RGB) {g=true, r=true}
  n28 = Panner(
   Coordinate=MaterialFunctionCall() {materialFunction=&VectorToRadialValue}.Radial Coordinates
   Speed=VectorParameter() {parameterName=noise speed, defaultValue=<rgba 0 -0.02 -1 1>}.RGB
  ) {speedX=0.2, speedY=1}
  n29 = Constant()
  n30 = TextureSample(UVs=Multiply(A=Multiply(A=@n27, B=@n28), B=Constant2Vector() {g=0.5, r=1}), "DDX(UVs)"=@n29, "DDY(UVs)"=@n29) {
   mipValueMode=TMVM_Derivative
   samplerType=SAMPLERTYPE_LinearGrayscale
   texture=&T_SimplePerlin_Greyscale
  }
  n31 = Add(A=Add(A=TextureCoordinate(), B=Constant2Vector() {r=-0.05}), B=Multiply(A=Multiply(A=@n30.R) {constB=0.3}) {constB=0.2})
  n32 = Add(A=Multiply(A=Add(A=@n24, B=@n23), B=Multiply(A=@n26, B=@n31)), B=ScalarParameter() {defaultValue=0.352, parameterName=Param_4})
  n33 = Clamp(None=@n32, Max=ScalarParameter() {defaultValue=1.2, parameterName=Param_5})
  n34 = Power(
   Base=SphereMask(A=TextureCoordinate(), B=Constant2Vector() {g=0.532, r=0.5}, Radius=ScalarParameter() {defaultValue=0.6, parameterName=Param_1}) {attenuationRadius=0.5, hardnessPercent=0}
   Exp=ScalarParameter() {defaultValue=1.8312, parameterName=Opacity_OutterMAsk_Pwr}
  )
  n35 = OneMinus(None=SphereMask(A=TextureCoordinate(), B=Constant2Vector() {g=0.5, r=0.5}, Radius=ScalarParameter() {defaultValue=0.1}) {attenuationRadius=0.3, hardnessPercent=1})
  n36 = ScalarParameter() {defaultValue=3.0871, parameterName=Opacity_InnerMask_Pwr}
  n37 = ScalarParameter() {defaultValue=0.8, parameterName=Param_7}
  n38 = ScalarParameter() {defaultValue=1, parameterName=Param_8}
  n39 = ScalarParameter() {defaultValue=1, parameterName=Param_9}
  n40 = ScalarParameter() {defaultValue=0.7, parameterName=Param_10}
  n41 = Multiply(
   A=LinearInterpolate(B=LinearInterpolate(B=@n33, Alpha=@n34), Alpha=Power(Base=@n35, Exp=@n36) {constExponent=1})
   B=Multiply(
    A=MaterialFunctionCall(Value_Morning=@n37, Value_Day=@n38, Value_Evening=@n39, Value_Night=@n40) {materialFunction=&Scalar_Time_of_Day}.Result
    B=ScalarParameter() {defaultValue=0.52, parameterName=Opacity_OverallMult}
   )
  )
  n42 = DepthFade(Opacity=@n41, FadeDistance=ScalarParameter() {defaultValue=196, parameterName=FadeDistance})
  n43 = QualitySwitch(Default=@n42, Low=@n41, High=@n42, Medium=@n42, Epic=@n42)
  n44 = NamedRerouteUsage()
  NamedRerouteDeclaration(None=Multiply(A=Multiply(A=@n43, B=@n44), B=NamedRerouteUsage())) {name=Opacity, nodeColor=<rgba 0.8516 0 1 1>}
  n45 = PixelDepth()
  n46 = Constant()
  n47 = ScalarParameter() {defaultValue=20000, parameterName=Param_6}
  NamedRerouteDeclaration(None=OneMinus(None=SphereMask(A=@n45.R, B=@n46, Radius=@n47) {hardnessPercent=0})) {name=FX_DistanceFade, nodeColor=<rgba 1 0 0.3047 1>}
  n48 = Constant() {r=0.16}
  n49 = Constant()
  n50 = SphereMask(A=Add(A=TextureCoordinate(), B=AppendVector(A=@n48, B=@n49)), B=Constant() {r=0.5}) {attenuationRadius=0.07, hardnessPercent=1}
  n51 = Constant() {r=0.35}
  n52 = Constant() {r=0.3}
  n53 = SphereMask(A=Add(A=TextureCoordinate(), B=AppendVector(A=@n51, B=@n52)), B=Constant() {r=0.5}) {attenuationRadius=0.25, hardnessPercent=0}
  n54 = Multiply(A=Power(Base=OneMinus(None=@n50)) {constExponent=9}, B=Power(Base=OneMinus(None=@n53)) {constExponent=11})
  n55 = Add(A=TextureCoordinate(), B=AppendVector(A=Constant(), B=Constant() {r=-0.4}))
  n56 = Constant() {r=0.5}
  n57 = Power(Base=OneMinus(None=SphereMask(A=@n55, B=@n56) {attenuationRadius=0.12, hardnessPercent=1})) {constExponent=4}
  NamedRerouteDeclaration(None=Saturate(None=Multiply(A=@n54, B=@n57))) {name=Mask, nodeColor=<rgba 1 0 0.3984 1>}
  n58 = SphereMask(A=TextureCoordinate(), B=Constant2Vector() {g=0.5, r=0.5}) {attenuationRadius=0.5, hardnessPercent=0}
  n59 = OneMinus(None=@n58)
  n60 = ScalarParameter() {defaultValue=2, parameterName=OpacityPowerCenter}
  n61 = ScalarParameter() {defaultValue=5, parameterName=Opacity}
  Reroute(None=Multiply(A=Power(Base=@n59, Exp=@n60), B=Multiply(A=@n58, B=@n61)))
  Saturate(None=@n32)
  SphereMask(A=TextureCoordinate(), B=Constant2Vector() {g=0.5, r=0.5}, Radius=ScalarParameter() {defaultValue=0.1, parameterName=s}) {attenuationRadius=0.3, hardnessPercent=1}
 ]
}
