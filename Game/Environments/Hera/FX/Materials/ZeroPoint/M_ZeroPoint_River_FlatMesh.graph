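# M_ZeroPoint_River_FlatMesh is a material in the Hera environment's Zero Point FX set, named for a flat river mesh.
# Its graph lists no expressions, so the file shows no textures, parameters or shading logic.
# It uses the Default Lit shading model and disables automatic setting of usage flags in the editor.
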Material /Game/Environments/Hera/FX/Materials/ZeroPoint/M_ZeroPoint_River_FlatMesh
graph {expressions {}, properties {bAutomaticallySetUsageInEditor=false, usedShadingModels=Default Lit}}
